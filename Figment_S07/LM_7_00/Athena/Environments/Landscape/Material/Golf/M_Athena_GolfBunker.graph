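# Material for golf course sand bunker meshes, blending sand with grass colored to match the landscape fairway.
# Vertex blue height-blends bunker UV sand with brightened, desaturated arid sand, then an asset mask tints the result low to high.
# The bunker asset normal is angle-corrected into the sand normal.
# Vertex green height-blends in the grass, and roughness is forced to 1 on ES2 and ES3.1.
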Material /Figment_S07/LM_7_00/Athena/Environments/Landscape/Material/Golf/M_Athena_GolfBunker
graph {
 properties {
  bTangentSpaceNormal=false
  bUseMaterialAttributes=true
  pixelDepthOffsetMode=PDOM_Legacy
  refractionMethod=RM_IndexOfRefraction
  usedShadingModels=Default Lit
 }
 nodes [
  n1 = ComponentMask(None=WorldPosition().XYZ) {b=true}
  n2 = ScalarParameter() {defaultValue=50, parameterName=TintGradientOffset}
  Clamp(None=Divide(A=Add(A=@n1, B=@n2), B=ScalarParameter() {defaultValue=50, parameterName=TintGradientRange}))
  n3 = Abs(None=ScalarParameter() {defaultValue=512, parameterName=TexScale})
  ComponentMask(None=Divide(A=WorldPosition() {worldPositionShaderOffset=WPT_ExcludeAllShaderOffsets}.XYZ, B=Multiply(A=@n3) {constB=-1})) {g=true, r=true}
  n4 = MaterialFunctionCall!(In, UseRoughness) {materialFunction=/Figment_S07/LM_7_00/Athena/Environments/Landscape/MaterialFunctions/Arid/MF_Athena_Arid_Sand}
  n5 = MaterialFunctionCall!(Material=@n4.Simple_Material) {materialFunction=/Engine/Functions/MaterialLayerFunctions/MatLayerBlend_BreakNormal}
  n6 = ScalarParameter() {parameterName=AridSandNormalFlatten}
  n7 = ScalarParameter() {defaultValue=1.25, parameterName=SandBrightness}
  n8 = MaterialFunctionCall!(
   In=MaterialFunctionCall!(
    Material=@n4.Simple_Material
    Normal=MaterialFunctionCall(Normal=@n5.Normal, Flatness=@n6) {materialFunction=&FlattenNormal}.Result
   ) {materialFunction=/Engine/Functions/MaterialLayerFunctions/MatLayerBlend_ReplaceNormals}.Result
   Color=@n7
  ) {materialFunction=/Figment_S07/LM_7_00/Athena/Environments/Landscape/MaterialFunctions/MF_Athena_Material_BaseMultiply}
  n9 = MaterialFunctionCall!(Material=@n8.Result) {materialFunction=/Engine/Functions/MaterialLayerFunctions/MatLayerBlend_BreakBaseColor}
  n10 = Desaturation(None=@n9.BaseColor, Fraction=ScalarParameter() {defaultValue=0.2, parameterName=SandBottomDesat})
  n11 = VertexColor()
  n12 = MaterialFunctionCall() {materialFunction=/Figment_S07/LM_7_00/Athena/Environments/Landscape/MaterialFunctions/Golf/MF_Athena_Golf_SandBunker_01}
  n13 = OneMinus(None=MaterialFunctionCall!(Mat attributes=@n12.UVSand) {materialFunction=/Engine/Functions/Engine_MaterialFunctions02/MatLayerBlend_BreakOpacity}.Result)
  n14 = ScalarParameter() {parameterName=BottomSandContrast}
  n15 = MaterialFunctionCall(
   In=MaterialFunctionCall!(
    Base Material=@n12.UVSand
    Top Material=MaterialFunctionCall!(Mat_In=@n8.Result, BaseColor=@n10) {materialFunction=/Figment_S07/LM_7_00/Athena/Environments/Landscape/MaterialFunctions/MF_Athena_ReplaceBaseColor}.Result
    Alpha=MaterialFunctionCall(Transition Phase=@n11.B, Height Texture=@n13, Contrast=@n14) {materialFunction=&HeightLerp}.Alpha
   ) {materialFunction=/Engine/Functions/MaterialLayerFunctions/MatLayerBlend_Standard}.Blended Material
   Color=LinearInterpolate(
    A=VectorParameter() {parameterName=LowTint, defaultValue=<rgba 0.75 0.75 0.75 0>}.RGB
    B=VectorParameter() {parameterName=HighTint, defaultValue=<rgba 1.25 1.25 1.25 0>}.RGB
    Alpha=@n12.BunkerAssetMask
   )
  ) {materialFunction=&MF_Athena_Material_BaseMultiply}
  n16 = MaterialFunctionCall(
   BaseNormal=Transform(None=@n12.BunkerAssetNormal)
   AdditionalNormal=MaterialFunctionCall(Material=@n15.Result) {materialFunction=&MatLayerBlend_BreakNormal}.Normal
  ) {materialFunction=&BlendAngleCorrectedNormals}
  n17 = MaterialFunctionCall() {materialFunction=/Figment_S07/LM_7_00/Athena/Environments/Landscape/MaterialFunctions/Standard/MF_Athena_Grass_01}
  n18 = MaterialFunctionCall!(
   In_LowMobile_Material
   Golf_Fairway_Mask=ScalarParameter() {defaultValue=0.75, desc=Compensate for layer shadow in landscape mat, parameterName=GolfMasks}
   In=@n17.Result
  ) {
   desc=Match fairway on landscape
   materialFunction=/Figment_S07/LM_7_00/Athena/Environments/Landscape/MaterialFunctions/MF_LawnGrassColoration_Smesh
  }
  n19 = MaterialFunctionCall(
   Transition Phase=VertexColor().G
   Height Texture=MaterialFunctionCall(Mat attributes=@n17.Result) {materialFunction=&MatLayerBlend_BreakOpacity}.Result
   Contrast=ScalarParameter() {defaultValue=1, parameterName=GrassBlendContrast}
  ) {materialFunction=&HeightLerp}
  n20 = BreakMaterialAttributes!(Attr=MaterialFunctionCall(
    Base Material=MaterialFunctionCall(Material=@n15.Result, Normal=@n16.Result) {materialFunction=&MatLayerBlend_ReplaceNormals}.Result
    Top Material=@n18.Result
    Alpha=OneMinus(None=@n19.Alpha)
   ) {materialFunction=&MatLayerBlend_Standard}.Blended Material)
  n21 = Constant() {r=1}
  MakeMaterialAttributes!(
   BaseColor=@n20.BaseColor
   Metallic
   Specular=@n20.BaseColor
   Roughness=FeatureLevelSwitch(Default=@n20.Roughness, ES2=@n21, ES3_1=@n21, SM4_REMOVED=@n20.Roughness, SM5=@n20.Roughness, SM6=@n20.Roughness)
   Anisotropy
   EmissiveColor
   Opacity=@n20.BaseColor
   OpacityMask
   Normal=@n20.BaseColor
   Tangent
   WorldPositionOffset=@n20.BaseColor
   SubsurfaceColor
   ClearCoat
   ClearCoatRoughness
   AmbientOcclusion
   Refraction
   CustomizedUVs_0
   CustomizedUVs_1
   CustomizedUVs_2
   CustomizedUVs_3
   CustomizedUVs_4
   CustomizedUVs_5
   CustomizedUVs_6
   CustomizedUVs_7
   PixelDepthOffset
   ShadingModel
   Displacement
  )
  Multiply(A=@n9.BaseColor, B=@n7)
 ]
}
